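; GA_Feat_S11_Revived is a gameplay ability for the Season 11 "Revived" feat, reporting revive achievement progress to the quest system.
; When activated from a gameplay event, it runs the parent logic, then gets the quest manager from the activating pawn's player controller.
; It sends a complex custom stat event with the Athena.Achievements.Revived tag and a count of one.
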(event Ability|EventActivateAbilityFromEvent
  (params (EventData :type Gameplay%20Event%20Data%20Structure%20%28by%20ref%29))
  (|CallParentFunction EventData)
  (Quest|SendComplexCustomStatEvent (Quests|GetQuestManager (Fortnite|GetFortPlayerControllerfromActor (Ability|GetActivatingPawn)) "Invalid") 0 "(GameplayTags=)" "(GameplayTags=((TagName=\"Athena.Achievements.Revived\")))" 1))
