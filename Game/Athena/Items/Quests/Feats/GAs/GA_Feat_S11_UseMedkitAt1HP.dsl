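; Gameplay ability for the Season 11 "use medkit at 1 HP" feat, flagging when the owning player is at the threshold.
; On activation it calls the parent, then waits on threshold changes for a configurable attribute, compared as less than or equal to 1.
; It applies a configurable gameplay effect to the player's own ASC on a match and removes it otherwise.
; Medkit use itself is not detected here.
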(event Ability|EventActivateAbility
  (|CallParentFunction)
  (bind (_asynctaskproxy _bmatchescomparison) (Ability|Tasks|WaitForAttributeChangeThreshold (Variables|Default|GetAttribute) "LessThanOrEqualTo" 1.0)
    (:OnChange
      (if _bmatchescomparison
        (GameplayEffects|ApplyGameplayEffectToTarget :self (Ability|GetAbilitySystemComponent (Variables|GAVars|Data|GetOwningPlayerPawn)) :GameplayEffectClass (Variables|Default|GetGE_Class) :Target (Ability|GetAbilitySystemComponent (Variables|GAVars|Data|GetOwningPlayerPawn)) :Level 0.0)
        (else
          (GameplayEffects|RemoveActiveGameplayEffectbySourceEffect (Ability|GetAbilitySystemComponent (Variables|GAVars|Data|GetOwningPlayerPawn)) (Variables|Default|GetGE_Class) 0 -1))))))

(event Ability|EventFailedToActivatePassiveAbility
  (params (ActorInfo :type Gameplay%20Ability%20Actor%20Info%20Structure)))
